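# Niagara system for the idle glow effect on the BikeMold pickaxe.
# AirGlowLightsBlue burst-spawns two spawn groups of blue glow sprites, using group masks to vary offset, scale and alignment; it writes light attributes but has no light renderer.
# BladeRimGlow burst-spawns into two mesh renderers, with its dynamic material parameter and time-of-day alpha modules disabled.
# Both emitters keep particles alive past their lifetime.
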NiagaraSystem /BRCosmetics/Weapons/FORT_Melee/Pickaxe_BikeMold/FX/NS_Pickaxe_BikeMold_Idle
dependencies {
 usedDataInterfaces []
 usedDynamicInputs [&MaskLinearColorBySpawnGroup]
 usedModules [&SystemState]
}
emitters {
 AirGlowLightsBlue {
  emitterUpdateScript [
   &EmitterState()
   &SpawnBurst_Instantaneous(Use Spawn Probability=false, UseLoopCountLimit=true)
   &SpawnBurst_Instantaneous(Spawn Group=1, Use Spawn Probability=false, UseLoopCountLimit=true) {module=SpawnBurst_Instantaneous001}
  ]
  particleSpawnScript [
   &InitializeParticle(
    Color Mode=::NewEnumerator0
    Position Offset=&MaskVectorBySpawnGroup(Masked Vector=<-15 0 5>, Passthrough Vector=<-15 0 -5>)
    UsePositionOffset=true
    Write Material Random=true
   )
   /SetVariables(
    Particles.AirGlowColor:LinearColor=&MultiplyLinearColor(Linear Color=<rgba 0 0.5833 1 1>, Scale Factor=<1 1 1 0.12>)
    Particles.AirGlowSpriteScale:Vector2f=&MaskVector2dBySpawnGroup(Masked Vector2D=<50 115>, Passthrough Vector2D=<50 120>, Spawn Group Mask=0)
   ) {isSetParameters=true}
  ]
  particleUpdateScript [
   &ParticleState(Kill Particles When Lifetime Has Elapsed=false, Let Infinitely Lived Particles Die When Emitter Deactivates=true)
   &SpriteFacingAndAlignment(
    Sprite Alignment=&MaskVectorBySpawnGroup(Masked Vector=<1 0 -0.15>, Passthrough Vector=<1 0 0.2>, Spawn Group Mask=1)
    WriteAlignment=true
    WriteFacing=false
   )
   &FN_Light_Attributes(
    Color=<rgba 0.2603 1 0.9901 1>
    Create Color Attribute=true
    Create Exponent Attribute=true
    Create Radius Attribute=true
    Exponent=3
    Radius=80
    Scale Color=2
   )
  ]
  rendererClasses [&NiagaraSpriteRendererProperties]
  renderers [{rendererClass=&NiagaraSpriteRendererProperties}]
 }
 BladeRimGlow {
  emitterUpdateScript [
   &EmitterState()
   &SpawnBurst_Instantaneous(Use Spawn Probability=false, UseLoopCountLimit=true)
  ]
  particleSpawnScript [&InitializeParticle(Color Mode=::NewEnumerator0, Lifetime=2, Position Mode=::NewEnumerator0, Write Material Random=true)]
  particleUpdateScript [
   &ParticleState(Kill Particles When Lifetime Has Elapsed=false, Let Infinitely Lived Particles Die When Emitter Deactivates=true)
   &DynamicMaterialParameters(
    Index 0 Param 1=&Multiply_Float(A=$Emitter.Age, B=-0.5)
    Module.Param0WriteEnabled=true
    Module.Param1WriteEnabled=true
    Module.Param2WriteEnabled=true
    Module.Param3WriteEnabled=true
   ) {enabled=false}
   &ScaleColor(
    Scale Alpha=&FNTimeOfDayScalar(Night Scale=0.08, Morning Scale=0.16, Day Scale=0.16, Evening Scale=0.08)
    ScaleA=true
    ScaleRGB=false
   ) {enabled=false}
  ]
  rendererClasses [&NiagaraMeshRendererProperties]
  renderers [
   {rendererClass=&NiagaraMeshRendererProperties, rendererIndex=0}
   {rendererClass=&NiagaraMeshRendererProperties, rendererIndex=1}
  ]
 }
}
summary {
 systemName=NS_Pickaxe_BikeMold_Idle
 emitters [
  {emitterName=BladeRimGlow, rendererClasses [&NiagaraMeshRendererProperties]}
  {emitterName=AirGlowLightsBlue, rendererClasses [&NiagaraSpriteRendererProperties]}
 ]
 userVariables []
}
